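# Sound cue for the final impact of demolishing rubber objects.
# It picks one of three rubber impact waves with equal weighting, then passes it through a modulator for slight random pitch variation.
# It plays at reduced volume under the impacts sound class with medium-sized-object destruction attenuation.
# An object-impact concurrency set is assigned, alongside override values that cap voices and stop the oldest.
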SoundCue /Game/Sounds/Fort_Impact_Sounds/Demolition/Rubber/Demo_Rubber_ImpactFinal_Cue
root=Modulator_2
nodes [
 {id=Modulator_2, children [Random_33], properties {pitchMax=1.07, pitchMin=0.92}}
 {id=Random_33, children [WavePlayer_100, WavePlayer_101, WavePlayer_102], properties {weights [1,1,1]}}
 {id=WavePlayer_100, properties {soundWaveAssetPtr=&/Rubber_ImpactFinal01}}
 {id=WavePlayer_101, properties {soundWaveAssetPtr=&/Rubber_ImpactFinal03}}
 {id=WavePlayer_102, properties {soundWaveAssetPtr=&/Rubber_ImpactFinal02}}
]
properties {
 attenuationSettings=/Game/Sounds/Attenuation/Fort_Demolition_Destroy_Medium_Sized_Object_Att
 maxDistance=4000
 soundClassObject=&Fort_Impacts_PSM
 volumeMultiplier=0.34
 concurrencyOverrides {maxCount=3, resolutionRule=StopOldest}
 concurrencySet [&CONC_Object_Impacts_FourMax]
}
